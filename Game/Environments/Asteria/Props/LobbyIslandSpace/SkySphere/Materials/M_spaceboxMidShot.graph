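# Unlit space skybox material for the Asteria lobby island sky sphere.
# It samples a spacebox texture through camera-vector lat-long UVs, warped by panning cloud noise, then adds tiled star layers with colour tints, nebula masking and animated noise twinkle.
# Sky atmosphere light direction and disk luminance build a tinted sun halo added over that result.
# Stray nodes remain, including an input-less Custom node holding pasted PostProcessVolume actor text instead of shader code.
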Material /Game/Environments/Asteria/Props/LobbyIslandSpace/SkySphere/Materials/M_spaceboxMidShot
graph {
 properties {pixelDepthOffsetMode=PDOM_Legacy, shadingModel=MSM_Unlit, usedShadingModels=Unlit}
 nodes [
  n1 = DotProduct(A=Multiply(A=SkyAtmosphereLightDirection()) {constB=-1}, B=PixelNormalWS())
  n2 = Power(Base=@n1, Exp=ScalarParameter() {defaultValue=300, parameterName=Param_3}) {constExponent=500}
  n3 = ScalarParameter() {defaultValue=0.03, parameterName=SunHaloIndensity2}
  n4 = SkyAtmosphereLightDiskLuminance!(DiskAngularDiameterOverride)
  n5 = Power(Base=@n4, Exp=ScalarParameter() {defaultValue=1})
  n6 = ScalarParameter() {defaultValue=1, parameterName=Param_1}
  n7 = Multiply(A=Clamp(None=Divide(A=@n5, B=@n6), Max=ScalarParameter() {defaultValue=0.02, parameterName=Param_2})) {constB=5000}
  n8 = VectorParameter() {parameterName=SunHaloColor, defaultValue=<rgba 1 0.3165 0.1866 1>}
  n9 = TextureSample(UVs=NamedRerouteUsage()) {texture=/Game/Environments/Asteria/Props/LobbyIslandSpace/SkySphere/Textures/T_spaceBox_2k}
  n10 = Power(Base=@n9.RGB, Exp=Constant() {r=3}) {constExponent=4}
  n11 = NamedRerouteUsage()
  n12 = TextureSample(UVs=Multiply(A=@n11, B=Constant() {r=10})) {
   samplerSource=SSM_Wrap_WorldGroupSettings
   samplerType=SAMPLERTYPE_LinearColor
   texture=/Game/Environments/Asteria/Props/LobbyIslandSpace/SkySphere/Textures/T_SkyDome_Stars_Sharp
  }
  n13 = TextureSample(UVs=Multiply(A=@n11, B=Constant() {r=5})) {
   samplerSource=SSM_Wrap_WorldGroupSettings
   samplerType=SAMPLERTYPE_LinearColor
   texture=&T_SkyDome_Stars_Sharp
  }
  n14 = Max(A=@n12.R, B=@n13.R)
  n15 = Power(Base=@n9.RGB, Exp=Constant() {r=1}) {constExponent=4}
  n16 = NamedRerouteUsage()
  n17 = TextureSampleParameter2D(UVs=Multiply(A=@n16, B=ScalarParameter() {defaultValue=14, parameterName=Nebula star_size_small})) {
   parameterName=Param
   samplerType=SAMPLERTYPE_LinearColor
   texture=&T_SkyDome_Stars_Sharp
  }
  n18 = TextureSample(UVs=Multiply(A=@n16, B=ScalarParameter() {defaultValue=6, parameterName=Nebula star_size_middle})) {
   samplerSource=SSM_Wrap_WorldGroupSettings
   samplerType=SAMPLERTYPE_LinearColor
   texture=&T_SkyDome_Stars_Sharp
  }
  n19 = TextureSample(UVs=Multiply(A=@n16, B=ScalarParameter() {defaultValue=5, parameterName=Nebula star_size_big})) {
   samplerSource=SSM_Wrap_WorldGroupSettings
   samplerType=SAMPLERTYPE_LinearColor
   texture=&T_SkyDome_Stars_Sharp
  }
  n20 = Max(A=@n17.R, B=Multiply(A=Max(A=@n18.R, B=@n19.R), B=ScalarParameter() {parameterName=use Big Star}))
  n21 = LinearInterpolate(
   A=Multiply(A=Multiply(A=@n10, B=@n14), B=Constant() {r=50})
   B=Multiply(A=Multiply(A=@n15, B=@n20), B=ScalarParameter() {defaultValue=75, parameterName=nenula star_brightness})
   Alpha=Power(
    Base=TextureSample() {texture=/Game/Environments/Asteria/Props/LobbyIslandSpace/SkySphere/Textures/T_SkySphere_starmask}.R
    Exp=Constant() {r=1}
   ) {constExponent=3}
  )
  n22 = ScalarParameter() {defaultValue=30, group=Stars, parameterName=Stars-StarMaskTileB}
  n23 = NamedRerouteUsage()
  n24 = ScalarParameter() {defaultValue=0.5, group=Stars, parameterName=Stars-StarMaskTileBSpeed}
  n25 = TextureSample(UVs=Panner(
    Coordinate=Multiply(A=@n23, B=@n22)
    Time=MaterialFunctionCall(Speed=@n24) {materialFunction=&TimeWithSpeedVariable}.Time
   ) {speedX=0.1, speedY=0.1}) {
   samplerSource=SSM_Wrap_WorldGroupSettings
   samplerType=SAMPLERTYPE_LinearGrayscale
   texture=/Game/Global/Textures/Default/Tileable_Patterns/T_StormCloudDetail01_M
  }
  n26 = Power(Base=@n25.R, Exp=ScalarParameter() {defaultValue=8.3697, group=Stars, parameterName=Stars-StarMaskTileBPower})
  n27 = MaterialFunctionCall!(CameraVector) {materialFunction=/Engine/Functions/Engine_MaterialFunctions01/Texturing/CameraVectorToLatLongUV}
  n28 = VectorParameter() {parameterName=CloudDistortionSpeed, defaultValue=<rgba 0.01 0 0 1>}
  n29 = Panner(Coordinate=NamedRerouteUsage(), Speed=AppendVector(A=@n28.R, B=@n28.R)) {speedX=0.01}
  n30 = TextureSample(UVs=@n29) {
   samplerSource=SSM_Wrap_WorldGroupSettings
   samplerType=SAMPLERTYPE_LinearGrayscale
   texture=&T_StormCloudDetail01_M
  }
  n31 = ConstantBiasScale(None=@n30.R) {bias=-0.5, scale=2}
  n32 = ScalarParameter() {defaultValue=0.0035, parameterName=CloudDistortionAmount, sliderMin=0.001}
  n33 = TextureSample(UVs=Add(A=NamedRerouteDeclaration(None=@n27.Uvs) {name=InfiniteUV, nodeColor=<rgba 0.6406 1 0 1>}, B=Multiply(A=@n31, B=@n32) {constB=0.01})) {texture=&T_spaceBox_2k}
  n34 = Multiply(A=Power(Base=@n33.A) {constExponent=1.3}, B=Multiply(A=Multiply(A=@n10, B=@n33.A), B=Constant() {r=3}))
  n35 = LinearInterpolate(A=@n33.RGB, B=Constant3Vector(), Alpha=ScalarParameter() {defaultValue=0.75, parameterName=Brightness})
  n36 = ScalarParameter() {defaultValue=0.25, parameterName=saturation}
  n37 = Multiply(A=@n23, B=Constant() {r=5})
  n38 = TextureSample(UVs=@n37) {
   samplerSource=SSM_Wrap_WorldGroupSettings
   samplerType=SAMPLERTYPE_LinearColor
   texture=&T_SkyDome_Stars_Sharp
  }
  n39 = Multiply(
   A=Power(Base=@n38.R, Exp=ScalarParameter() {defaultValue=4.5, group=Stars, parameterName=Stars-StarPower})
   B=ScalarParameter() {defaultValue=1.5, group=Stars, parameterName=Stars-StarMod}
  )
  n40 = TextureSample(UVs=Multiply(A=@n23, B=ScalarParameter() {defaultValue=20, group=Stars, parameterName=Stars-StarMaskTileA})) {
   samplerSource=SSM_Wrap_WorldGroupSettings
   samplerType=SAMPLERTYPE_LinearGrayscale
   texture=&T_StormCloudDetail01_M
  }
  n41 = Power(Base=@n40.R, Exp=ScalarParameter() {defaultValue=6.6574, group=Stars, parameterName=Stars-StarMaskTileAPower})
  n42 = Multiply(
   A=Multiply(A=Multiply(A=@n39, B=@n41), B=@n26)
   B=LinearInterpolate(A=ScalarParameter() {defaultValue=86.0193, group=Stars, parameterName=Stars-StarBrightness}, B=Constant() {r=1051.9219}, Alpha=@n25.R)
  )
  n43 = TextureSample(UVs=Multiply(A=@n23, B=ScalarParameter() {defaultValue=10, group=Stars, parameterName=Stars-StarTile})) {
   samplerSource=SSM_Wrap_WorldGroupSettings
   texture=/Game/Environments/Asteria/Props/LobbyIslandSpace/SkySphere/Textures/T_SkyDome_Stars_Sharp_mask
  }
  n44 = TextureSample(UVs=@n37) {
   samplerSource=SSM_Wrap_WorldGroupSettings
   texture=&T_SkyDome_Stars_Sharp_mask
  }
  n45 = LinearInterpolate(A=@n43.RGB, B=@n44.RGB, Alpha=Constant() {r=0.284})
  n46 = LinearInterpolate(
   A=LinearInterpolate(
    A=Constant3Vector() {constant=<rgba 0 0.0473 1 0>}
    B=Constant3Vector() {constant=<rgba 1 0 0.083 0>}
    Alpha=ComponentMask(None=@n45) {r=true}
   )
   B=Constant3Vector() {constant=<rgba 0 1 0.0471 0>}
   Alpha=ComponentMask(None=@n45) {b=true}
  )
  n47 = Constant() {r=0.5}
  n48 = TextureSample(UVs=Multiply(A=TextureCoordinate(), B=Constant() {r=6.4248})) {
   samplerSource=SSM_Wrap_WorldGroupSettings
   samplerType=SAMPLERTYPE_LinearGrayscale
   texture=&T_StormCloudDetail01_M
  }
  n49 = Power(Base=@n48.R, Exp=Constant() {r=4})
  n50 = Add(
   A=MaterialFunctionCall!(In=Desaturation(None=@n35, Fraction=@n36), Contrast=ScalarParameter() {defaultValue=0.001, parameterName=Contrast}) {materialFunction=/Engine/Functions/Engine_MaterialFunctions01/ImageAdjustment/CheapContrast_RGB}.Result
   B=LinearInterpolate(A=Constant3Vector(), B=Multiply(A=@n42, B=@n46), Alpha=LinearInterpolate(A=@n47, Alpha=@n49))
  )
  n51 = NamedRerouteUsage()
  n52 = Multiply(A=@n51, B=ScalarParameter() {defaultValue=10, group=Stars, parameterName=Stars-StarTile})
  n53 = TextureSample(UVs=@n52) {
   samplerSource=SSM_Wrap_WorldGroupSettings
   samplerType=SAMPLERTYPE_LinearColor
   texture=&T_SkyDome_Stars_Sharp
  }
  n54 = ScalarParameter() {defaultValue=30, group=Stars, parameterName=Stars-StarMaskTileB}
  n55 = ScalarParameter() {defaultValue=0.5, group=Stars, parameterName=Stars-StarMaskTileBSpeed}
  n56 = TextureSample(UVs=Panner(
    Coordinate=Multiply(A=@n51, B=@n54)
    Time=MaterialFunctionCall(Speed=@n55) {materialFunction=&TimeWithSpeedVariable}.Time
   ) {speedX=0.1, speedY=0.1}) {
   samplerSource=SSM_Wrap_WorldGroupSettings
   samplerType=SAMPLERTYPE_LinearGrayscale
   texture=&T_StormCloudDetail01_M
  }
  n57 = Multiply(A=@n53.R, B=Power(Base=@n56.R, Exp=ScalarParameter() {defaultValue=8.3697, group=Stars, parameterName=Stars-StarMaskTileBPower}))
  n58 = TextureSample(UVs=@n52) {
   samplerSource=SSM_Wrap_WorldGroupSettings
   texture=&T_SkyDome_Stars_Sharp_mask
  }
  n59 = TextureSample(UVs=Multiply(A=NamedRerouteUsage(), B=ScalarParameter() {defaultValue=20, parameterName=Param_20})) {
   samplerSource=SSM_Wrap_WorldGroupSettings
   texture=&T_SkyDome_Stars_Sharp_mask
  }
  n60 = LinearInterpolate(A=@n58.RGB, B=@n59.RGB, Alpha=Constant() {r=0.5})
  n61 = LinearInterpolate(
   A=LinearInterpolate(
    A=Constant3Vector() {constant=<rgba 0 0.0473 1 0>}
    B=Constant3Vector() {constant=<rgba 1 0 0.083 0>}
    Alpha=ComponentMask(None=@n60) {r=true}
   )
   B=Constant3Vector() {constant=<rgba 0 1 0.0471 0>}
   Alpha=ComponentMask(None=@n60) {b=true}
  )
  n62 = Add(
   A=Add(A=Multiply(A=@n21, B=@n26), B=Add(A=@n34, B=@n50))
   B=Multiply(A=Multiply(A=@n57, B=@n61), B=ScalarParameter() {defaultValue=2.5, parameterName=BGstar_brightness})
  )
  Add(A=Add(A=Multiply(A=@n2, B=@n3) {constB=0.05}, B=Multiply(A=@n7, B=@n8.RGB)), B=Saturate(None=@n62))
  Add(A=@n62, B=SkyAtmosphereViewLuminance!(WorldDirection))
  ComponentMask(None=SkyAtmosphereAerialPerspective!(World Position)) {b=true, g=true, r=true}
  Constant()
  Custom() {
   code="Begin Map\r\n   Begin Level\r\n      Begin Actor Class=/Script/Engine.PostProcessVolume Name=PostProcessVolume_UAID_D45D64562520E03801_1606380761 Archetype=&PostProcessVolume'/Script/Engine.Default__PostProcessVolume' ActorFolderPath=None\r\n         Begin Object Class=/Script/Engine.BrushComponent Name=\"BrushComponent0\" Archetype=&BrushComponent'&Default__PostProcessVolume:BrushComponent0'\r\n            Begin Object Class=/Script/Engine.BodySetup Name=\"BodySetup_0\"\r\n            End Object\r\n         End Object\r\n         Begin Brush Name=Model_1\r\n         End Brush\r\n         Begin Object Class=/Script/UnrealEd.CubeBuilder Name=\"CubeBuilder_3\"\r\n         End Object\r\n         Begin Object Name=\"BrushComponent0\"\r\n            Begin Object Name=\"BodySetup_0\"\r\n               AggGeom=(ConvexElems=((VertexData=((X=-100.000000,Y=100.000000,Z=-100.000000),(X=-100.000000,Y=100.000000,Z=100.000000),(X=-100.000000,Y=-100.000000,Z=100.000000),(X=-100.000000,Y=-100.000000,Z=-100.000000),(X=100.000000,Y=100.000000,Z=-100.000000),(X=100.000000,Y=100.000000,Z=100.000000),(X=100.000000,Y=-100.000000,Z=-100.000000),(X=100.000000,Y=-100.000000,Z=100.000000)),IndexData=(0,4,6,2,0,3,0,6,3,6,2,3,4,0,1,0,2,1,2,6,7,6,4,7,4,1,5,1,7,5,7,4,5,1,2,7),ElemBox=(Min=(X=-100.000000,Y=-100.000000,Z=-100.000000),Max=(X=100.000000,Y=100.000000,Z=100.000000),IsValid=1))))\r\n               bGenerateMirroredCollision=False\r\n               CollisionTraceFlag=CTF_UseSimpleAsComplex\r\n            End Object\r\n            Brush=/Script/Engine.Model'\"Model_1\"'\r\n            BrushBodySetup=&BodySetup'\"BodySetup_0\"'\r\n            BodyInstance=(MaxAngularVelocity=3599.999756)\r\n            RelativeLocation=(X=-301372.000000,Y=-301104.000000,Z=500344.000000)\r\n         End Object\r\n         Begin Object Name=\"CubeBuilder_3\"\r\n         End Object\r\n         Settings=(bOverride_WhiteTemp=True,bOverride_WhiteTint=True,bOverride_ColorSaturation=True,bOverride_ColorContrast=True,bOverride_ColorGamma=True,bOverride_ColorContrastShadows=True,bOverride_ColorGammaShadows=True,bOverride_ColorSaturationHighlights=True,bOverride_ColorGainHighlights=True,bOverride_ColorCorrectionHighlightsMin=True,bOverride_ColorCorrectionHighlightsMax=True,bOverride_SceneFringeIntensity=True,bOverride_BloomMethod=True,bOverride_BloomIntensity=True,bOverride_Bloom2Tint=True,bOverride_Bloom3Tint=True,bOverride_Bloom4Tint=True,bOverride_Bloom5Tint=True,bOverride_Bloom6Tint=True,bOverride_BloomConvolutionSize=True,bOverride_BloomConvolutionPreFilterMin=True,bOverride_VignetteIntensity=True,WhiteTemp=5851.999512,ColorSaturation=(X=0.800000,Y=0.800000,Z=0.800000,W=1.000000),ColorContrast=(X=1.100000,Y=1.099281,Z=1.098614,W=1.000000),ColorGamma=(X=0.900000,Y=0.900000,Z=0.900000,W=1.000000),ColorGammaShadows=(X=0.800000,Y=0.800000,Z=0.800000,W=1.000000),ColorSaturationHighlights=(X=0.800000,Y=0.800000,Z=0.800000,W=1.000000),ColorGainHighlights=(X=2.000000,Y=2.000000,Z=2.000000,W=1.000000),ColorCorrectionHighlightsMin=0.000000,ColorCorrectionHighlightsMax=0.500000,SceneFringeIntensity=0.500000,BloomIntensity=0.500000,Bloom4Tint=(R=0.031250,G=0.031250,B=0.031250,A=1.000000),Bloom5Tint=(R=0.052083,G=0.052083,B=0.052083,A=1.000000),Bloom6Tint=(R=0.109266,G=0.154844,B=0.276042,A=1.000000),BloomConvolutionSize=40.000000,BloomConvolutionTexture=/Script/Engine.Texture2D'\"/Engine/EngineMaterials/DefaultBloomKernel\"',BloomConvolutionPreFilterMin=0.000000,VignetteIntensity=0.500000)\r\n         bUnbound=True\r\n         BrushType=Brush_Add\r\n         Begin Brush Name=Model_1\r\n            Begin PolyList\r\n               Begin Polygon\r\n                  Origin   -00100.000000,-00100.000000,-00100.000000\r\n                  Normal   -00001.000000,+00000.000000,+00000.000000\r\n                  TextureU +00000.000000,+00001.000000,+00000.000000\r\n                  TextureV +00000.000000,+00000.000000,-00001.000000\r\n                  Vertex   -00100.000000,-00100.000000,-00100.000000\r\n                  Vertex   -00100.000000,-00100.000000,+00100.000000\r\n                  Vertex   -00100.000000,+00100.000000,+00100.000000\r\n                  Vertex   -00100.000000,+00100.000000,-00100.000000\r\n               End Polygon\r\n               Begin Polygon\r\n                  Origin   -00100.000000,+00100.000000,-00100.000000\r\n                  Normal   +00000.000000,+00001.000000,+00000.000000\r\n                  TextureU +00001.000000,-00000.000000,+00000.000000\r\n                  TextureV +00000.000000,+00000.000000,-00001.000000\r\n                  Vertex   -00100.000000,+00100.000000,-00100.000000\r\n                  Vertex   -00100.000000,+00100.000000,+00100.000000\r\n                  Vertex   +00100.000000,+00100.000000,+00100.000000\r\n                  Vertex   +00100.000000,+00100.000000,-00100.000000\r\n               End Polygon\r\n               Begin Polygon\r\n                  Origin   +00100.000000,+00100.000000,-00100.000000\r\n                  Normal   +00001.000000,+00000.000000,+00000.000000\r\n                  TextureU +00000.000000,-00001.000000,+00000.000000\r\n                  TextureV +00000.000000,+00000.000000,-00001.000000\r\n                  Vertex   +00100.000000,+00100.000000,-00100.000000\r\n                  Vertex   +00100.000000,+00100.000000,+00100.000000\r\n                  Vertex   +00100.000000,-00100.000000,+00100.000000\r\n                  Vertex   +00100.000000,-00100.000000,-00100.000000\r\n               End Polygon\r\n               Begin Polygon\r\n                  Origin   +00100.000000,-00100.000000,-00100.000000\r\n                  Normal   +00000.000000,-00001.000000,+00000.000000\r\n                  TextureU -00001.000000,-00000.000000,-00000.000000\r\n                  TextureV +00000.000000,+00000.000000,-00001.000000\r\n                  Vertex   +00100.000000,-00100.000000,-00100.000000\r\n                  Vertex   +00100.000000,-00100.000000,+00100.000000\r\n                  Vertex   -00100.000000,-00100.000000,+00100.000000\r\n                  Vertex   -00100.000000,-00100.000000,-00100.000000\r\n               End Polygon\r\n               Begin Polygon\r\n                  Origin   -00100.000000,+00100.000000,+00100.000000\r\n                  Normal   +00000.000000,+00000.000000,+00001.000000\r\n                  TextureU +00001.000000,+00000.000000,+00000.000000\r\n                  TextureV +00000.000000,+00001.000000,+00000.000000\r\n                  Vertex   -00100.000000,+00100.000000,+00100.000000\r\n                  Vertex   -00100.000000,-00100.000000,+00100.000000\r\n                  Vertex   +00100.000000,-00100.000000,+00100.000000\r\n                  Vertex   +00100.000000,+00100.000000,+00100.000000\r\n               End Polygon\r\n               Begin Polygon\r\n                  Origin   -00100.000000,-00100.000000,-00100.000000\r\n                  Normal   +00000.000000,+00000.000000,-00001.000000\r\n                  TextureU +00001.000000,+00000.000000,+00000.000000\r\n                  TextureV +00000.000000,-00001.000000,+00000.000000\r\n                  Vertex   -00100.000000,-00100.000000,-00100.000000\r\n                  Vertex   -00100.000000,+00100.000000,-00100.000000\r\n                  Vertex   +00100.000000,+00100.000000,-00100.000000\r\n                  Vertex   +00100.000000,-00100.000000,-00100.000000\r\n               End Polygon\r\n            End PolyList\r\n         End Brush\r\n         Brush=&Model'\"Model_1\"'\r\n         BrushComponent=\"BrushComponent0\"\r\n         BrushBuilder=&CubeBuilder'\"CubeBuilder_3\"'\r\n         bHidden=False\r\n         RootComponent=\"BrushComponent0\"\r\n         ActorLabel=\"PostProcessVolume\"\r\n         bIsSpatiallyLoaded=False\r\n      End Actor\r\n   End Level\r\nBegin Surface\r\nEnd Surface\r\nEnd Map"
   inputs []
  }
  MaterialFunctionCall() {materialFunction=&RemapValue_0-1}
  Multiply(A=@n29)
  Multiply(A=@n4)
  Multiply()
  PixelDepth()
  Saturate(None=@n5)
  ScalarParameter() {defaultValue=0.0025, parameterName=CloudDistortionAmount}
  SphereMask(A=@n4, B=Constant() {r=1}) {attenuationRadius=1}
  Step(X=@n1) {constY=0.99}
  Subtract(A=Multiply(A=CameraPositionWS()) {constB=0.9}, B=ActorPositionWS())
  TextureCoordinate()
  TextureCoordinate()
  TextureCoordinate()
  TextureCoordinate()
  Time()
  VertexNormalWS()
 ]
}
